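# Physics asset for the SweetLetter pickaxe cosmetic, covering a metal ring charm on the weapon mesh.
# It has two capsule bodies, one on the base_metal_ring bone and a shorter one on dyn_metal_ring.
# A single constraint joins the dynamic ring to the base ring, with a wide limit angle on one swing axis, a narrower one on the other, and a near-zero twist limit.
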PhysicsAsset /BRCosmetics/Weapons/FORT_Melee/Pickaxe_SweetLetter/Meshes/SweetLetter_Axe_Physics
bodies {
 base_metal_ring [{shapeType=Capsule, length=2.005, radius=0.505, shapeName=base_metal_ring_capsule, rotation=<rot 0 -90 0>}]
 dyn_metal_ring [{shapeType=Capsule, length=0.5, radius=0.5, shapeName=dyn_charm_02_m_capsule, center=<1 0 0>, rotation=<rot 90 0 0>}]
}
constraints [{bone1Name=dyn_metal_ring, bone2Name=base_metal_ring, swing1LimitDegrees=30, swing2LimitDegrees=85, twistLimitDegrees=1}]
