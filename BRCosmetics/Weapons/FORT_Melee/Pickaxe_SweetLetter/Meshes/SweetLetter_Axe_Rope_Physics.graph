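# Physics asset for the rope on the SweetLetter pickaxe axe, with capsule bodies for the rope and part of a character's lower body.
# Top, middle and bottom dyn rope capsules each constrain to their own tiny base capsule with limited swing and twist, rather than chaining together.
# Pelvis, spine_01, both thighs and the right calf form a separate angularly free group, with three offset capsules on the calf.
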PhysicsAsset /BRCosmetics/Weapons/FORT_Melee/Pickaxe_SweetLetter/Meshes/SweetLetter_Axe_Rope_Physics
bodies {
 base_bottom_rope [{shapeType=Capsule, length=0.505, radius=0.505, shapeName=base_bottom_rope_capsule, rotation=<rot 0 -90 0>}]
 base_mid_rope [{shapeType=Capsule, length=0.505, radius=0.505, shapeName=base_mid_rope_capsule, rotation=<rot 0 -90 0>}]
 base_top_rope [{shapeType=Capsule, length=0.505, radius=0.505, shapeName=base_top_rope_capsule, rotation=<rot 0 -90 0>}]
 calf_r [
  {
   shapeType=Capsule
   length=26.0929
   radius=9.1166
   shapeName=thigh_r_capsule
   center=<26.3539 19.8981 -1.7618>
   rotation=<rot 85.2958 -4.836 0.0001>
  }
  {
   shapeType=Capsule
   length=26.0929
   radius=9.1166
   shapeName=thigh_r_capsule
   center=<24.7973 1.4999 -1.7618>
   rotation=<rot 85.2958 -4.836 0.0001>
  }
  {
   shapeType=Capsule
   length=19.0929
   radius=5.6166
   shapeName=thigh_r_capsule
   center=<21.4973 -13.1725 0.3714>
   rotation=<rot 85.2958 -4.836 0.0001>
  }
 ]
 dyn_bottom_rope [{shapeType=Capsule, length=8.5, radius=5, shapeName=dyn_bottom_rope_capsule, center=<7.9305 0 0>, rotation=<rot 90 0 0>}]
 dyn_mid_rope [{shapeType=Capsule, length=18, radius=5, shapeName=dyn_mid_rope_capsule, center=<17.7239 0 0>, rotation=<rot 90 0 0>}]
 dyn_top_rope [{shapeType=Capsule, length=18, radius=5, shapeName=dyn_top_rope_capsule, center=<13.0346 0 1.6228>, rotation=<rot 90 0 0>}]
 pelvis [{shapeType=Capsule, length=9.6815, radius=12.2728, shapeName=pelvis_capsule, center=<-1.8522 -1.386 0>, rotation=<rot 0 -90 0>}]
 spine_01 [{shapeType=Capsule, length=15.5573, radius=9.2327, shapeName=spine_01_capsule, center=<1.7545 0 0>, rotation=<rot 0 -90 0>}]
 thigh_l [{shapeType=Capsule, length=34.8554, radius=9.3541, shapeName=thigh_r_capsule, center=<-22.1907 -0.6066 2.52>, rotation=<rot 90 0 0>}]
 thigh_r [{shapeType=Capsule, length=34.6616, radius=9.5479, shapeName=thigh_r_capsule, center=<22.1907 0.6066 -0.3009>, rotation=<rot 90 0 0>}]
}
constraints [
 {bone1Name=dyn_top_rope, bone2Name=base_top_rope, swing1LimitDegrees=50, swing2LimitDegrees=25, twistLimitDegrees=10}
 {bone1Name=dyn_mid_rope, bone2Name=base_mid_rope, swing1LimitDegrees=45, swing2LimitDegrees=45, twistLimitDegrees=10}
 {bone1Name=dyn_bottom_rope, bone2Name=base_bottom_rope, swing1LimitDegrees=70, swing2LimitDegrees=40, twistLimitDegrees=0}
 {
  bone1Name=spine_01
  bone2Name=pelvis
  swing1LimitDegrees=45
  swing1Motion=Free
  swing2LimitDegrees=45
  swing2Motion=Free
  twistLimitDegrees=45
  twistMotion=Free
 }
 {
  bone1Name=thigh_l
  bone2Name=pelvis
  swing1LimitDegrees=45
  swing1Motion=Free
  swing2LimitDegrees=45
  swing2Motion=Free
  twistLimitDegrees=45
  twistMotion=Free
 }
 {
  bone1Name=thigh_r
  bone2Name=pelvis
  swing1LimitDegrees=45
  swing1Motion=Free
  swing2LimitDegrees=45
  swing2Motion=Free
  twistLimitDegrees=45
  twistMotion=Free
 }
 {
  bone1Name=calf_r
  bone2Name=thigh_r
  swing1LimitDegrees=45
  swing1Motion=Free
  swing2LimitDegrees=45
  swing2Motion=Free
  twistLimitDegrees=45
  twistMotion=Free
 }
]
